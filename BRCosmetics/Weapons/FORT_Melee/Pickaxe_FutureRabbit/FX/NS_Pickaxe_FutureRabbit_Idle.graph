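# Idle Niagara effect for the FutureRabbit pickaxe, emitting drifting sprite shards from a weapon socket while it is not swinging.
# User.bIsSwinging zeroes the spawn rate, User.ShardsSpawn scales spawn probability, and System.FrontEndScale scales sprite size and launch speed.
# Shards drift on curl noise and drag, rotating and fading, with User.Variant selecting among six color palettes with random hue shift.
# Two sprite renderers share the emitter, which also computes per-particle glow color and size.
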NiagaraSystem /BRCosmetics/Weapons/FORT_Melee/Pickaxe_FutureRabbit/FX/NS_Pickaxe_FutureRabbit_Idle
dependencies {
 usedDataInterfaces [
  &NiagaraDataInterfaceCamera
  &NiagaraDataInterfaceCurve
  &NiagaraDataInterfaceColorCurve
  &NiagaraDataInterfaceSocketReader
 ]
 usedDynamicInputs [&FN_FrontEndFloatModifier]
 usedModules [
  &self:SystemSpawnScript.SystemScriptSource.SystemScriptGraph.NiagaraNodeAssignment_1.SetVariables_#1
  &self:SystemSpawnScript.SystemScriptSource.SystemScriptGraph.NiagaraNodeAssignment_0.SetVariables_#2
  &SystemState
 ]
}
emitters {Shards {
  emitterSpawnScript [/SetVariables(Emitter.PatternColor:LinearColor=&MergeHelper_ColorSwitch(
     0=<rgba 1 0.4735 0.0999 1>
     1=<rgba 0.0999 1 1 1>
     2=<rgba 0.6087 0.4975 0.2213 1>
     3=<rgba 0.5382 0.1 1 1>
     4=<rgba 0.0999 1 0.7011 1>
     5=<rgba 0.0999 1 1 1>
     Index=&MakeIntFromFloat(Float=$User.Variant)
     Num=::NewEnumerator4
    )) {isSetParameters=true}]
  emitterUpdateScript [
   &EmitterState()
   &SpawnRate(
    Spawn Probability=&Lerp_Float(A=0, B=0.66, Alpha=$User.ShardsSpawn)
    SpawnRate=&MakeCustomFloatFromBool(Bool=$User.bIsSwinging, True Float=0, False Float=1)
    Use Spawn Probability=true
   )
  ]
  particleSpawnScript [
   &InitializeParticle(
    Color=<rgba 1 1 1 0.6957>
    Lifetime Max=1
    Lifetime Min=0.65
    Lifetime Mode=::NewEnumerator1
    Mesh Scale Mode=::NewEnumerator1
    Mesh Uniform Scale Max=1.2
    Mesh Uniform Scale Min=0.7
    Sprite Size=&Multiply_Vector2D_ByFloat(Float=$System.FrontEndScale, Vector2D=&RandomRangeVector2D(Minimum=<1.6 8>, Maximum=<4 12>))
    Sprite Size Mode=::NewEnumerator4
    UsePositionOffset=false
    Write Material Random=true
   )
   &InheritVelocity(Inherited Velocity Speed Limit=150, Speed Limit=true, UseSpeedThreshold=false)
   &SocketLocation(
    "Inherit Velocity (Scale)"=0.3
    SamplePosOffset=true
    Sampled Position Offset=<-2.5 0 40>
    Socket Reader=$System.Socket Reader
    Socket Sampling Mode=::NewEnumerator3
    VelocitySpeedLimit=false
   )
   &ShapeLocation(MaskSpawnGroup=false, Non Uniform Scale=<1 1 4.5>, Override Seed=false, Sphere Radius=5, Sphere Surface Distribution=0.5, UseOwnerScale=true)
   &SpriteFacingAndAlignment(
    Sprite Alignment=$Engine.Owner.SystemZAxis
    Sprite Facing=&ReturnCameraVector(Camera Vectors=::NewEnumerator1, Coordinate Space=::Simulation)
    WriteAlignment=true
    WriteFacing=false
   )
   &AddVelocity(
    Velocity=&RandomRangeVector(
     Minimum=&RandomRangeVector(Minimum=<-1 -1 0>, Maximum=<-0.5 -0.5 0>)
     Maximum=&RandomRangeVector(Minimum=<0.5 0.5 2>, Maximum=<1 1 3>)
    )
    Velocity Speed Scale=&Multiply_Float(A=15, B=$System.FrontEndScale)
   )
  ]
  particleUpdateScript [
   &CurlNoiseForce(
    Noise Frequency=12
    Noise Quality / Cost=::NewEnumerator0
    Noise Strength=&RandomRangeFloat(Minimum=15, Maximum=35)
    Pan Noise=true
    Pan Noise Field=<0.02 0.02 0.02>
   )
   &Drag(Use Linear Drag=true, Use Rotational Drag=true)
   &SolveForcesAndVelocity(Clamp Velocity=false, Limit Acceleration=false)
   &ParticleState()
   &SpriteRotationRate(Rotation Rate=&FloatFromCurve(
     FloatCurve=DataInterface() {propertyValues {Curve {
        defaultValue=1
        keys [{interpMode=RCIM_Cubic, tangentMode=RCTM_User, time=0.15, value=0}, {interpMode=RCIM_Cubic, tangentMode=RCTM_User, time=1, value=1}]
       }}}
     Scale Curve=&RandomRangeFloat(
      Minimum=&RandomRangeFloat(Minimum=-100, Maximum=-200)
      Maximum=&RandomRangeFloat(Minimum=100, Maximum=200)
     )
    ))
   &ScaleColor(
    Scale Alpha=&FloatFromCurve(FloatCurve=DataInterface() {propertyValues {Curve {keys [
         {arriveTangent=7.7281, interpMode=RCIM_Cubic, leaveTangent=7.7281, tangentMode=RCTM_User, time=0, value=0}
         {arriveTangent=1.3333, leaveTangent=1.3333, time=0.15, value=1}
         {interpMode=RCIM_Cubic, tangentMode=RCTM_Break, time=0.75, value=1}
         {arriveTangent=-8.6393, tangentMode=RCTM_User, time=1, value=0}
        ]}}})
    Scale RGB=&MakeVectorFromLinearColorRGB(Color=&HueShiftLinearColor(
      Hue=&RandomRangeFloat(Minimum=-0.1, Maximum=0.1)
      Saturation=1
      Value=&RandomRangeFloat(Minimum=0.6, Maximum=1)
      Linear Color=&MergeHelper_ColorSwitch(
       0=&ColorFromCurve(ColorCurve=DataInterface() {propertyValues {
          AlphaCurve {keys [<at 0 1 RCIM_Cubic>,<at 1 0 RCIM_Cubic>]}
          BlueCurve {keys [<at 0.5 0.1 RCIM_Cubic>,<at 1 0.1 RCIM_Cubic>]}
          GreenCurve {keys [<at 0.5 0.8469 RCIM_Cubic>,<at 1 0.475 RCIM_Cubic>]}
          RedCurve {keys [<at 0.5 1 RCIM_Cubic>,<at 1 1 RCIM_Cubic>]}
         }})
       1=&ColorFromCurve(ColorCurve=DataInterface() {propertyValues {
          AlphaCurve {keys [<at 0 1 RCIM_Cubic>,<at 1 0 RCIM_Cubic>]}
          BlueCurve {keys [<at 0.5 1 RCIM_Cubic>,<at 1 1 RCIM_Cubic>]}
          GreenCurve {keys [<at 0.5 1 RCIM_Cubic>,<at 1 1 RCIM_Cubic>]}
          RedCurve {keys [<at 0.5 0.1 RCIM_Cubic>,<at 1 0.1 RCIM_Cubic>]}
         }})
       2=&ColorFromCurve(ColorCurve=DataInterface() {propertyValues {
          AlphaCurve {keys [<at 0 1 RCIM_Cubic>,<at 1 0 RCIM_Cubic>]}
          BlueCurve {keys [<at 0.5 0.2195 RCIM_Cubic>,<at 1 0 RCIM_Cubic>]}
          GreenCurve {keys [<at 0.5 0.4969 RCIM_Cubic>,<at 1 0 RCIM_Cubic>]}
          RedCurve {keys [<at 0.5 0.6105 RCIM_Cubic>,<at 1 0 RCIM_Cubic>]}
         }})
       3=&ColorFromCurve(ColorCurve=DataInterface() {propertyValues {
          AlphaCurve {keys [<at 0 1 RCIM_Cubic>,<at 1 0 RCIM_Cubic>]}
          BlueCurve {keys [<at 0.5 1 RCIM_Cubic>,<at 1 1 RCIM_Cubic>]}
          GreenCurve {keys [<at 0.5 1 RCIM_Cubic>,<at 1 0.4 RCIM_Cubic>]}
          RedCurve {keys [<at 0.5 0.1 RCIM_Cubic>,<at 1 0.6922 RCIM_Cubic>]}
         }})
       4=&ColorFromCurve(ColorCurve=DataInterface() {propertyValues {
          AlphaCurve {keys [<at 0 1 RCIM_Cubic>,<at 1 0 RCIM_Cubic>]}
          BlueCurve {keys [<at 0.5 0.5 RCIM_Cubic>,<at 1 0.7>]}
          GreenCurve {keys [<at 0.5 0.5 RCIM_Cubic>,<at 1 1>]}
          RedCurve {keys [<at 0.5 0.05 RCIM_Cubic>,<at 1 0.1>]}
         }})
       5=&ColorFromCurve(ColorCurve=DataInterface() {propertyValues {
          AlphaCurve {keys [<at 0 1 RCIM_Cubic>,<at 1 0 RCIM_Cubic>]}
          BlueCurve {keys [<at 0.5 0.5 RCIM_Cubic>,<at 1 1>]}
          GreenCurve {keys [<at 0.5 0.0891 RCIM_Cubic>,<at 1 0.1>]}
          RedCurve {keys [<at 0.5 0.05 RCIM_Cubic>,<at 1 0.6322>]}
         }})
       Index=&MakeIntFromFloat(Float=$User.Variant)
       Num=::NewEnumerator4
      )
     ))
    ScaleA=true
    ScaleRGB=true
   )
   /SetVariables(
    Particles.GlowColor=&MultiplyLinearColor(
     Linear Color=$Particles.Color
     Scale Factor=&MakeVector4_FromFloat(Float=&FloatFromCurve(
       FloatCurve=DataInterface() {propertyValues {Curve {
          defaultValue=1
          keys [
           {tangentMode=RCTM_User, time=0, value=1}
           {arriveTangent=-0.0405, interpMode=RCIM_Cubic, leaveTangent=-0.0405, tangentMode=RCTM_User, time=0.3697, value=1}
           {arriveTangent=-3.7213, interpMode=RCIM_Cubic, leaveTangent=-3.7213, tangentMode=RCTM_User, time=1, value=0}
          ]
         }}}
       Scale Curve=0.5
      ))
    )
    Particles.GlowSize=&Multiply_Vector2D(A=$Particles.SpriteSize, B=<2.5 1.6>)
   ) {isSetParameters=true}
  ]
  rendererClasses [&NiagaraSpriteRendererProperties]
  renderers [
   {rendererClass=&NiagaraSpriteRendererProperties}
   {rendererClass=&NiagaraSpriteRendererProperties, rendererIndex=1}
  ]
 }}
summary {
 systemName=NS_Pickaxe_FutureRabbit_Idle
 emitters [{emitterName=Shards, rendererClasses [&NiagaraSpriteRendererProperties]}]
 userVariables [
  User.bIsFrontEnd {type=NiagaraBool}
  User.bIsFrontEndPreview {type=NiagaraBool}
  User.bIsSwinging {type=NiagaraBool}
  User.RibbonIDOffset {type=NiagaraInt32}
  User.ShardsSpawn {type=NiagaraFloat, defaultValue=1, description=""}
  User.Variant {type=NiagaraFloat, defaultValue=0, description=""}
 ]
}
